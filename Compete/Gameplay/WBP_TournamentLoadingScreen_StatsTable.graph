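# Tournament loading screen stats table widget showing total points above rows for Victory Royales, elims, placement and matches played.
# SetupStats checks validity, then copies TournamentLoadingStatsVM values into the total text and reused leaderboard stats item widgets.
# The Elims and Placement rows receive elimination and placement points rather than raw counts.
# PreConstruct sets the gradient background material's corner radius according to whether the game is mobile.
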WidgetBlueprint /Compete/Gameplay/WBP_TournamentLoadingScreen_StatsTable
graphs [
 EventGraph {nodes [
   Event PreConstruct()
   > n1 = GetDynamicMaterial(self=Get BackgroundGradient() -> (BackgroundGradient:Image Object Reference))
   > SetVectorParameterValue(
    self=@n1
    ParameterName=Corner Radius
    Value=Select(Option 0:Linear Color Structure=(R=10,G=10,B=10,A=10), Option 1:Linear Color Structure=(R=25,G=25,B=25,A=25), Index=IsMobileGame()) -> (ReturnValue:Linear Color Structure)
   )
  ]}
 SetupStats {nodes [
   n1 = SetupStats!() -> (then, IsValid:bool=~) {is=title K2Node_FunctionEntry}
   > Branch(Condition=@n1)
   n2 = Get TournamentLoadingStatsVM() -> (TournamentLoadingStatsVM:Tournament Loading Stats VM Object Reference)
   > Set Text(
    Text="To Text (Integer)"(Value=Get TotalScore!(self:Tournament Loading Stats VM Object Reference=@n2) -> (TotalScore:int=0))
    self=Get GainedPoints() -> (GainedPoints:Common Text Object Reference)
   )
   > SetValueFromInt!(
    execute
    self:WBP Tournament Leaderboard Stats Item Object Reference=Get MatchesPlayed() -> (MatchesPlayed:WBP Tournament Leaderboard Stats Item Object Reference)
    Value:int=Get MatchesPlayed!(self:Tournament Loading Stats VM Object Reference=@n2) -> (MatchesPlayed:int=0)
   ) -> (then)
   > SetValueFromInt(
    self=Get Placement() -> (Placement:WBP Tournament Leaderboard Stats Item Object Reference)
    Value=Get PlacementPoints!(self:Tournament Loading Stats VM Object Reference=@n2) -> (PlacementPoints:int=0)
   )
   > SetValueFromInt(
    self=Get Elims() -> (Elims:WBP Tournament Leaderboard Stats Item Object Reference)
    Value=Get EliminationPoints!(self:Tournament Loading Stats VM Object Reference=@n2) -> (EliminationPoints:int=0)
   )
   > SetValueFromInt(
    self=Get VictoryRoyales() -> (VictoryRoyales:WBP Tournament Leaderboard Stats Item Object Reference)
    Value=Get NumVictoryRoyales!(self:Tournament Loading Stats VM Object Reference=@n2) -> (NumVictoryRoyales:int=0)
   )
  ]}
]
variables []
widgets {
 info {inheritedWidgetCount=0, parentClass=&CommonUserWidget, rootWidgetClass=&GridPanel, widgetCount=16}
 widgets [
  Root {widgetClassPath=&GridPanel, widget=Root, properties {flowDirectionPreference=Culture, columnFill [1]}}
  CaptureForPostBufferUpdate_1 {
   widgetClassPath=&WBP_CaptureForPostBufferUpdate
   parent=Root
   widget=CaptureForPostBufferUpdate_1
   slot {class=GridSlot}
  }
  Backplate {
   widgetClassPath=&WBP_UIKit_Backplate
   parent=Root
   widget=Backplate
   properties {brightness=Dark}
   slot {class=GridSlot, properties {layer=1}}
  }
  PointsBackground {
   widgetClassPath=&Image
   parent=Root
   widget=PointsBackground
   properties {
    renderOpacity=0
    brush {
     bIsDynamicallyLoaded=false
     drawAs=Image
     imageType=FullColor
     mirroring=NoMirror
     resourceName=None
     resourceObject=null
     tiling=NoTile
     imageSize=<584 160>
     margin=<ltrb 0 0 0 0>
     outlineSettings {
      bUseBrushTransparency=false
      roundingType=FixedRadius
      width=0
      color {colorUseRule=UseColor_Specified, specifiedColor=<rgba 0 0 0 0>}
      cornerRadii=<10 10 10 10>
     }
     tintColor {colorUseRule=UseColor_Specified, specifiedColor=<rgba 1 1 1 1>}
     uVRegion {bIsValid=false, max=<0 0>, min=<0 0>}
    }
   }
   slot {class=GridSlot, properties {layer=2}}
  }
  BackgroundGradient {
   widgetClassPath=&Image
   bIsVariable=true
   parent=Root
   widget=BackgroundGradient
   properties {
    visibility=SelfHitTestInvisible
    brush {
     bIsDynamicallyLoaded=false
     drawAs=Image
     imageType=FullColor
     mirroring=NoMirror
     resourceName=None
     resourceObject=/Compete/MaterialInstances/MI_UI_Gradient
     tiling=NoTile
     imageSize=<32 32>
     margin=<ltrb 0 0 0 0>
     outlineSettings {
      bUseBrushTransparency=false
      roundingType=HalfHeightRadius
      width=0
      color {colorUseRule=UseColor_Specified, specifiedColor=<rgba 0 0 0 0>}
      cornerRadii=<0 0 0 0>
     }
     tintColor {colorUseRule=UseColor_Specified, specifiedColor=<rgba 1 1 1 1>}
     uVRegion {bIsValid=false, max=<0 0>, min=<0 0>}
    }
    colorAndOpacity=<rgba 1 1 1 0.5>
   }
   slot {class=GridSlot, properties {rowSpan=2}}
  }
  TotalPoints {
   widgetClassPath=&HorizontalBox
   parent=Root
   widget=TotalPoints
   slot {class=GridSlot, properties {layer=3, verticalAlignment=VAlign_Center, padding=<ltrb 24 0 24 2>}}
  }
  TotalPoints_Text {
   widgetClassPath=&CommonTextBlock
   parent=TotalPoints
   widget=TotalPoints_Text
   properties {
    text="Total  points"
    wrapTextAt=172
    colorAndOpacity {colorUseRule=UseColor_Specified, specifiedColor=<rgba 0.5776 0.5776 0.5776 1>}
    font {
     bForceMonospaced=false
     bMaterialIsStencil=false
     fontMaterial=null
     fontObject=&HeadingNow
     letterSpacing=0
     monospacedWidth=1
     size=22.5
     skewAmount=0
     typefaceFontName=75 Medium
     outlineSettings {
      bApplyOutlineToDropShadows=false
      bMiteredCorners=false
      bSeparateFillAlpha=false
      outlineMaterial=null
      outlineSize=0
      outlineColor=<rgba 0 0 0 1>
     }
    }
    shadowOffset=<0 0>
   }
   slot {class=HorizontalBoxSlot, properties {horizontalAlignment=HAlign_Center, verticalAlignment=VAlign_Center}}
  }
  GainedPoints {
   widgetClassPath=&CommonTextBlock
   bIsVariable=true
   parent=TotalPoints
   widget=GainedPoints
   properties {
    justification=Center
    text=---
    font {
     bForceMonospaced=false
     bMaterialIsStencil=false
     fontMaterial=null
     fontObject=&HeadingNow
     letterSpacing=-40
     monospacedWidth=1
     size=41.25
     skewAmount=0
     typefaceFontName=86 Bold Italic
     outlineSettings {
      bApplyOutlineToDropShadows=false
      bMiteredCorners=false
      bSeparateFillAlpha=false
      outlineMaterial=null
      outlineSize=0
      outlineColor=<rgba 0 0 0 1>
     }
    }
    shadowOffset=<0 0>
   }
   slot {class=HorizontalBoxSlot, properties {horizontalAlignment=HAlign_Right, verticalAlignment=VAlign_Center, size {sizeRule=Fill, value=1}}}
  }
  StatsContainer {
   widgetClassPath=&VerticalBox
   parent=Root
   widget=StatsContainer
   properties {bOverride_Cursor=true}
   slot {class=GridSlot, properties {layer=3, row=1, padding=<ltrb 16 42 16 75>}}
  }
  VictoryRoyales {
   widgetClassPath=/Compete/TournamentLeaderboard/WBP_TournamentLeaderboardStatsItem
   bIsVariable=true
   parent=StatsContainer
   widget=VictoryRoyales
   properties {isLodaingSceen=true, valueName=Victory Royales}
   slot {class=VerticalBoxSlot, properties {padding=<ltrb 8 0 8 0>}}
  }
  Separator_1 {
   widgetClassPath=&Image
   parent=StatsContainer
   widget=Separator_1
   properties {
    brush {
     bIsDynamicallyLoaded=false
     drawAs=Image
     imageType=NoImage
     mirroring=NoMirror
     resourceName=None
     resourceObject=null
     tiling=NoTile
     imageSize=<32 1>
     margin=<ltrb 0 0 0 0>
     outlineSettings {
      bUseBrushTransparency=false
      roundingType=HalfHeightRadius
      width=0
      color {colorUseRule=UseColor_Specified, specifiedColor=<rgba 0 0 0 0>}
      cornerRadii=<0 0 0 0>
     }
     tintColor {colorUseRule=UseColor_Specified, specifiedColor=<rgba 1 1 1 1>}
     uVRegion {bIsValid=false, max=<0 0>, min=<0 0>}
    }
    colorAndOpacity=<rgba 1 1 1 0.502>
   }
   slot {class=VerticalBoxSlot, properties {padding=<ltrb 0 25 0 0>}}
  }
  Elims {
   widgetClassPath=&WBP_TournamentLeaderboardStatsItem
   bIsVariable=true
   parent=StatsContainer
   widget=Elims
   properties {isLodaingSceen=true, valueName=Elims}
   slot {class=VerticalBoxSlot, properties {padding=<ltrb 8 25 8 0>}}
  }
  Separator_2 {
   widgetClassPath=&Image
   parent=StatsContainer
   widget=Separator_2
   properties {
    brush {
     bIsDynamicallyLoaded=false
     drawAs=Image
     imageType=NoImage
     mirroring=NoMirror
     resourceName=None
     resourceObject=null
     tiling=NoTile
     imageSize=<32 1>
     margin=<ltrb 0 0 0 0>
     outlineSettings {
      bUseBrushTransparency=false
      roundingType=HalfHeightRadius
      width=0
      color {colorUseRule=UseColor_Specified, specifiedColor=<rgba 0 0 0 0>}
      cornerRadii=<0 0 0 0>
     }
     tintColor {colorUseRule=UseColor_Specified, specifiedColor=<rgba 1 1 1 1>}
     uVRegion {bIsValid=false, max=<0 0>, min=<0 0>}
    }
    colorAndOpacity=<rgba 1 1 1 0.502>
   }
   slot {class=VerticalBoxSlot, properties {padding=<ltrb 0 25 0 0>}}
  }
  Placement {
   widgetClassPath=&WBP_TournamentLeaderboardStatsItem
   bIsVariable=true
   parent=StatsContainer
   widget=Placement
   properties {isLodaingSceen=true, valueName=Placement}
   slot {class=VerticalBoxSlot, properties {padding=<ltrb 8 25 8 0>}}
  }
  Separator_3 {
   widgetClassPath=&Image
   parent=StatsContainer
   widget=Separator_3
   properties {
    brush {
     bIsDynamicallyLoaded=false
     drawAs=Image
     imageType=NoImage
     mirroring=NoMirror
     resourceName=None
     resourceObject=null
     tiling=NoTile
     imageSize=<32 1>
     margin=<ltrb 0 0 0 0>
     outlineSettings {
      bUseBrushTransparency=false
      roundingType=HalfHeightRadius
      width=0
      color {colorUseRule=UseColor_Specified, specifiedColor=<rgba 0 0 0 0>}
      cornerRadii=<0 0 0 0>
     }
     tintColor {colorUseRule=UseColor_Specified, specifiedColor=<rgba 1 1 1 1>}
     uVRegion {bIsValid=false, max=<0 0>, min=<0 0>}
    }
    colorAndOpacity=<rgba 1 1 1 0.502>
   }
   slot {class=VerticalBoxSlot, properties {padding=<ltrb 0 25 0 0>}}
  }
  MatchesPlayed {
   widgetClassPath=&WBP_TournamentLeaderboardStatsItem
   bIsVariable=true
   parent=StatsContainer
   widget=MatchesPlayed
   properties {isLodaingSceen=true, valueName=Matches Played}
   slot {class=VerticalBoxSlot, properties {padding=<ltrb 8 25 8 0>}}
  }
 ]
}
