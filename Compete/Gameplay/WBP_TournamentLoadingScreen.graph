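# Tournament loading screen widget, a CommonUserWidget that lays out tournament details, region and stats.
# An Overlay layers the shared tournament background beneath a SafeZone grid holding the details header at top left, a region label at top right and a StatsTable sub-widget below the header.
# The graph has only empty event stubs and no variables, so this Blueprint itself does not populate the child widgets.
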WidgetBlueprint /Compete/Gameplay/WBP_TournamentLoadingScreen
graphs [EventGraph {nodes [
   Event PreConstruct()
   Event Construct()
   Event Tick()
  ]}]
variables []
widgets {
 info {inheritedWidgetCount=0, parentClass=&CommonUserWidget, rootWidgetClass=&Overlay, widgetCount=7}
 widgets [
  Root {widgetClassPath=&Overlay, widget=Root}
  Background {
   widgetClassPath=/Compete/TournamentDetails/WBP_TournamentBackground
   bIsVariable=true
   parent=Root
   widget=Background
   slot {class=OverlaySlot, properties {horizontalAlignment=HAlign_Fill, verticalAlignment=VAlign_Fill}}
  }
  SafeZone {
   widgetClassPath=&SafeZone
   parent=Root
   widget=SafeZone
   slot {class=OverlaySlot, properties {horizontalAlignment=HAlign_Fill, verticalAlignment=VAlign_Fill}}
  }
  Main {widgetClassPath=&GridPanel, parent=SafeZone, widget=Main, properties {columnFill [1], rowFill [0]}, slot {class=SafeZoneSlot}}
  CommonText_Region {
   widgetClassPath=&CommonTextBlock
   bIsVariable=true
   parent=Main
   widget=CommonText_Region
   properties {
    scrollStyle=&TextScrollStyle-Base_C
    text=NAE
    font {
     bForceMonospaced=false
     bMaterialIsStencil=false
     fontMaterial=null
     fontObject=&HeadingNow
     letterSpacing=0
     monospacedWidth=1
     size=18
     skewAmount=0
     typefaceFontName=75 Medium
     outlineSettings {
      bApplyOutlineToDropShadows=false
      bMiteredCorners=false
      bSeparateFillAlpha=false
      outlineMaterial=null
      outlineSize=2
      outlineColor=<rgba 0 0 0 0.2>
     }
    }
    shadowOffset=<0 0>
   }
   slot {class=GridSlot, properties {column=1, horizontalAlignment=HAlign_Right, verticalAlignment=VAlign_Top, padding=<ltrb 0 90 80 0>}}
  }
  Header {
   widgetClassPath=/Compete/TournamentDetails/WBP_TournamentDetailsHeader
   bIsVariable=true
   parent=Main
   widget=Header
   slot {class=GridSlot, properties {horizontalAlignment=HAlign_Left, verticalAlignment=VAlign_Top, padding=<ltrb 80 69 0 0>}}
  }
  StatsTable {
   widgetClassPath=&/WBP_TournamentLoadingScreen_StatsTable
   bIsVariable=true
   parent=Main
   widget=StatsTable
   slot {class=GridSlot, properties {horizontalAlignment=HAlign_Left, row=1, verticalAlignment=VAlign_Top, padding=<ltrb 80 60 0 0>}}
  }
 ]
}
